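# Sound cue for the Creative toy player scoreboard, playing the left-side score sound.
# A single wave player feeds Player_Scoreboard_ScoreLeft straight to the output in the Fort_Impacts sound class, with no randomization or mixing.
# Inline attenuation overrides the referenced gadget-activate preset with a large falloff shape.
# Overridden concurrency allows one instance per owner, so a new play stops the older one.
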SoundCue /Game/Sounds/Creative/Toys/Scoreboard/Player_Scoreboard_ScoreLeft_Cue
root=WavePlayer_0
nodes [{id=WavePlayer_0, properties {soundWaveAssetPtr=&/Player_Scoreboard_ScoreLeft}}]
properties {
 attenuationSettings=&Fort_GadgetActivate_Att
 bOverrideAttenuation=true
 bOverrideConcurrency=true
 maxDistance=23600
 soundClassObject=&Fort_Impacts
 volumeMultiplier=1
 attenuationOverrides {attenuationShapeExtents {x=20000}}
 concurrencyOverrides {bLimitToOwner=true, maxCount=1, resolutionRule=StopOldest}
}
